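# Unlit translucent Niagara material for the breaking effect in the main player's Psyche heart-sign emote.
# A panning distortion normal map warps both a scrolling heart-sign pattern and a procedural noise field.
# The pattern, masked by a custom shape, is multiplied by particle color and a time-of-day tint for emissive.
# Opacity combines particle alpha, the shape mask and contrasted noise offset by a Dissolve dynamic parameter.
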Material /Game/Animation/Game/MainPlayer/Emotes/HeartSign/FX/M_Psyche_Heart_Sign_Breaks
graph {
 properties {
  bUsedWithNiagaraMeshParticles=true
  bUsedWithNiagaraRibbons=true
  bUsedWithNiagaraSprites=true
  blendMode=BLEND_Translucent
  pixelDepthOffsetMode=PDOM_Legacy
  shadingModel=MSM_Unlit
  usedShadingModels=Unlit
 }
 nodes [
  n1 = ParticleColor()
  n2 = TextureCoordinate()
  n3 = Panner() {speedX=-0.1, speedY=-0.2}
  n4 = Saturate(None=Multiply(A=TextureSample(UVs=@n3) {samplerType=SAMPLERTYPE_Normal, texture=&T_Weap_Distort_N_02}.RGB) {constB=0.35})
  n5 = Panner(Coordinate=Add(A=Multiply(A=@n2), B=ComponentMask(None=@n4) {b=true})) {speedX=0.1, speedY=0.2}
  n6 = TextureSample() {texture=&/T_Custom_Shape}
  n7 = Multiply(A=Clamp(None=TextureSample(UVs=@n5) {texture=&/T_Psyche_Heart_Sign_Pattern}.RGB), B=@n6.R)
  n8 = Constant3Vector() {constant=<rgba 0.25 0.25 0.25 0>}
  n9 = Constant3Vector() {constant=<rgba 0.9 0.9 0.75 0>}
  n10 = Constant3Vector() {constant=<rgba 0.2 0.2 0.1 0>}
  n11 = Constant3Vector() {constant=<rgba 0.8 0.8 0.5 0>}
  MaterialFunctionCall(Emissive=Multiply(
    A=Multiply(A=@n1.RGB, B=@n7)
    B=MaterialFunctionCall(Color_Morning=@n8, Color_Day=@n9, Color_Evening=@n10, Color_Night=@n11) {materialFunction=&Color_Time_of_Day}.Result
   )) {materialFunction=&MF_FX_EmissiveAdjustments}
  n12 = MaterialFunctionCall(Float2=TextureCoordinate()) {materialFunction=&BreakOutFloat2Components}
  n13 = Add(
   A=MaterialFunctionCall(X=@n12.R, Y=@n12.R) {materialFunction=&MakeFloat3}.Result
   B=ComponentMask(None=@n4) {b=true}
  )
  n14 = MaterialFunctionCall(
   In=Add(A=Noise!(World Position=@n13, FilterWidth) {levels=0, scale=7}, B=DynamicParameter() {paramNames [Dissolve,Param2,Param3,Param4]}.Dissolve)
   Contrast=Constant() {r=5}
  ) {materialFunction=&CheapContrast}
  Multiply(A=@n1.A, B=Multiply(A=@n6.R, B=Multiply(A=@n6.R, B=@n14.Result) {constB=0}))
 ]
}
