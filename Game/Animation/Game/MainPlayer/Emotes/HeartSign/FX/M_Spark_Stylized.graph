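# Masked unlit stylized spark material for Niagara sprite, mesh and ribbon particles, filed under the main player's HeartSign emote FX.
# Color_Time_of_Day takes morning, day, evening and night tints, and its result is multiplied by particle color and fed into MF_FX_EmissiveAdjustments.
# A custom shape texture, sampled at an explicit mip level, has its green channel doubled and multiplied by particle alpha.
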Material /Game/Animation/Game/MainPlayer/Emotes/HeartSign/FX/M_Spark_Stylized
graph {
 properties {
  bUsedWithNiagaraMeshParticles=true
  bUsedWithNiagaraRibbons=true
  bUsedWithNiagaraSprites=true
  blendMode=BLEND_Masked
  pixelDepthOffsetMode=PDOM_Legacy
  shadingModel=MSM_Unlit
  usedShadingModels=Unlit
 }
 nodes [
  n1 = Constant3Vector() {constant=<rgba 0.25 0.25 0.25 0>}
  n2 = Constant3Vector() {constant=<rgba 0.9 0.9 0.75 0>}
  n3 = Constant3Vector() {constant=<rgba 0.2 0.2 0.1 0>}
  n4 = Constant3Vector() {constant=<rgba 0.8 0.8 0.5 0>}
  n5 = ParticleColor()
  MaterialFunctionCall(Emissive=Multiply(
    A=MaterialFunctionCall(Color_Morning=@n1, Color_Day=@n2, Color_Evening=@n3, Color_Night=@n4) {materialFunction=&Color_Time_of_Day}.Result
    B=@n5.RGB
   )) {materialFunction=&MF_FX_EmissiveAdjustments}
  n6 = TextureSample!(UVs, Tex, Level, Apply View MipBias) {mipValueMode=TMVM_MipLevel, texture=&/T_Custom_Shape}
  Multiply(A=Add(A=@n6.G, B=@n6.G), B=@n5.A)
 ]
}
