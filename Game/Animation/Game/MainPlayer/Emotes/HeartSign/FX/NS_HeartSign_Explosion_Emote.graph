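# Niagara effect for the main player's HeartSign emote: a heart that rises in pulses, then bursts larger with sparks.
# Heart_Shape and glow Flare001 share a height curve, with hand-keyed scale beats and a midlife surge; heart material parameters are also animated.
# Spark emitters, some GPU-simulated, add start and mid-sequence bursts plus a trail following the rise via normalized loop age.
# A Light emitter is included; Flare and Halo are disabled.
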NiagaraSystem /Game/Animation/Game/MainPlayer/Emotes/HeartSign/FX/NS_HeartSign_Explosion_Emote
dependencies {
 usedDataInterfaces [
  &NiagaraDataInterfaceCurve
  &NiagaraDataInterfaceVectorCurve
  &NiagaraDataInterfaceColorCurve
  &NiagaraDataInterfaceVector2DCurve
 ]
 usedDynamicInputs [&Vector2DFromFloat]
 usedModules [&SystemState]
}
emitters {
 Flare {
  bEnabled=false
  emitterUpdateScript [
   &EmitterState()
   &SpawnBurst_Instantaneous(Spawn Time=0.1, Use Spawn Probability=false, UseLoopCountLimit=false)
  ]
  particleSpawnScript [&InitializeParticle(
    Color=<rgba 5 0.25 0.25 0.2>
    Lifetime=2.4
    Position Offset=<0 0 12.5>
    Sprite Rotation Mode=::NewEnumerator2
    Sprite Size Mode=::NewEnumerator3
    Uniform Sprite Size=35
    UsePositionOffset=true
    Write Material Random=true
   ) {module=InitializeParticle001}]
  particleUpdateScript [
   &ParticleState()
   &ScaleColor(
    Scale Alpha=&FloatFromCurve(FloatCurve=DataInterface() {propertyValues {Curve {keys [
         {interpMode=RCIM_Cubic, tangentMode=RCTM_User, time=0, value=0}
         {interpMode=RCIM_Cubic, tangentMode=RCTM_User, time=0.5, value=1}
         {tangentMode=RCTM_User, time=1, value=0}
        ]}}})
    ScaleA=true
    ScaleRGB=true
   )
   &ScaleSpriteSize(Uniform Curve Sprite Scale=DataInterface() {propertyValues {Curve {keys [
        {arriveTangent=2.1973, interpMode=RCIM_Cubic, leaveTangent=2.1973, tangentMode=RCTM_User, time=0, value=0}
        {arriveTangent=1, interpMode=RCIM_Cubic, leaveTangent=1, time=0.1389, value=0.75}
        {arriveTangent=0.0963, interpMode=RCIM_Cubic, leaveTangent=0.0963, tangentMode=RCTM_User, time=1, value=1}
       ]}}})
   &CameraOffset(Camera Offset Amount=5)
   &SpriteRotationRate(Rotation Rate=25)
  ]
  rendererClasses [&NiagaraSpriteRendererProperties]
  renderers [{rendererClass=&NiagaraSpriteRendererProperties}]
 }
 Flare001 {
  emitterUpdateScript [
   &EmitterState(Life Cycle Mode=::NewEnumerator1, Recalculate Duration Each Loop=true, UseLoopDelay=true)
   &SpawnBurst_Instantaneous(Use Spawn Probability=false, UseLoopCountLimit=false)
  ]
  particleSpawnScript [&InitializeParticle(
    Color=<rgba 1 1 1 0.75>
    Lifetime=5
    Position Offset=<0 0 12.5>
    Position Offset Coordinate Space=::Local
    Sprite Size Mode=::NewEnumerator3
    Uniform Sprite Size=35
    UsePositionOffset=true
    Write Material Random=true
   ) {module=InitializeParticle001}]
  particleUpdateScript [
   &ParticleState()
   &ScaleColor(
    Linear Color Curve=DataInterface() {propertyValues {
      AlphaCurve {keys [<at 0 0>,<at 0.0077 0.5>,<at 0.2506 0.25>,<at 0.4178 0.5>,<at 0.5849 0>]}
      BlueCurve {keys [<at 0.0038 0>,<at 0.2697 0.75>,<at 0.4736 0>,<at 0.6542 0.75>,<at 0.8513 0>]}
      GreenCurve {keys [<at 0.0038 0.25>,<at 0.2697 0>,<at 0.4736 0.25>,<at 0.6542 0>,<at 0.8513 0.25>]}
      RedCurve {keys [<at 0.0038 1>,<at 0.2697 1>,<at 0.4736 1>,<at 0.6542 1>,<at 0.8513 2>]}
     }}
    Scale Mode=::NewEnumerator2
   )
   /SetVariables(Particles.Position=&Add_Vector(
     A=Linked() {linkedVariable {name=Particles.Initial.Position, type=NiagaraPosition}}
     B=&VectorFromCurve(VectorCurve=DataInterface() {propertyValues {
        XCurve {keys [<at 0 0>,<at 1 0>]}
        YCurve {keys [<at 0 0>,<at 1 0>]}
        ZCurve {keys [
          {arriveTangent=-29.7738, interpMode=RCIM_Cubic, leaveTangent=-29.7738, tangentMode=RCTM_User, time=0, value=0.8097}
          {arriveTangent=428.0804, interpMode=RCIM_Cubic, leaveTangent=428.0804, tangentMode=RCTM_User, time=0.0605, value=11.5343}
          {arriveTangent=25.3279, interpMode=RCIM_Cubic, leaveTangent=25.3279, tangentMode=RCTM_User, time=0.1502, value=10.5442}
          {arriveTangent=334.4314, interpMode=RCIM_Cubic, leaveTangent=334.4314, tangentMode=RCTM_User, time=0.2222, value=27.8481}
          {arriveTangent=25.038, interpMode=RCIM_Cubic, leaveTangent=25.038, time=0.3079, value=22.6366}
          {arriveTangent=148.1974, interpMode=RCIM_Cubic, leaveTangent=148.1974, tangentMode=RCTM_User, time=0.3986, value=32.2649}
          {arriveTangent=-552.3597, interpMode=RCIM_Cubic, leaveTangent=-552.3597, tangentMode=RCTM_User, time=0.4953, value=21.4375}
          {arriveTangent=817.7939, interpMode=RCIM_Cubic, leaveTangent=817.7939, tangentMode=RCTM_User, time=0.5052, value=22.3636}
          {arriveTangent=880.1193, interpMode=RCIM_Cubic, leaveTangent=880.1193, tangentMode=RCTM_User, time=0.512, value=26.6862}
          {arriveTangent=532.9858, interpMode=RCIM_Cubic, leaveTangent=532.9858, tangentMode=RCTM_User, time=0.5247, value=38.8145}
          {arriveTangent=123.7225, interpMode=RCIM_Cubic, leaveTangent=123.7225, tangentMode=RCTM_User, time=0.5502, value=44.8436}
          {arriveTangent=0.0095, interpMode=RCIM_Cubic, leaveTangent=0.0095, tangentMode=RCTM_User, time=1, value=100}
         ]}
       }})
    )) {isSetParameters=true}
   &ScaleSpriteSize(
    Uniform Curve Scale=1.5
    Uniform Curve Sprite Scale=DataInterface() {propertyValues {Curve {keys [
        {arriveTangent=-0.2072, interpMode=RCIM_Cubic, leaveTangent=-0.2072, tangentMode=RCTM_User, time=0, value=0}
        {
         arriveTangent=12.3238
         arriveTangentWeight=0.0288
         interpMode=RCIM_Cubic
         leaveTangent=12.3238
         tangentMode=RCTM_User
         time=0.0864
         value=0.5651
        }
        {
         arriveTangent=1.8531
         arriveTangentWeight=0.0219
         interpMode=RCIM_Cubic
         leaveTangent=1.8531
         tangentMode=RCTM_User
         time=0.152
         value=0.5679
        }
        {
         arriveTangent=14.9653
         arriveTangentWeight=0.023
         interpMode=RCIM_Cubic
         leaveTangent=14.9653
         tangentMode=RCTM_User
         time=0.2209
         value=0.8144
        }
        {
         arriveTangent=-0.9078
         arriveTangentWeight=0.0321
         interpMode=RCIM_Cubic
         leaveTangent=-0.9078
         tangentMode=RCTM_User
         time=0.3172
         value=0.7307
        }
        {
         arriveTangent=15.9193
         arriveTangentWeight=0.0296
         interpMode=RCIM_Cubic
         leaveTangent=15.9193
         tangentMode=RCTM_User
         time=0.4061
         value=1.5175
        }
        {
         arriveTangent=-24.9715
         arriveTangentWeight=0.0231
         interpMode=RCIM_Cubic
         leaveTangent=-24.9715
         tangentMode=RCTM_User
         time=0.4755
         value=1.2207
        }
        {
         arriveTangent=-21.6376
         arriveTangentWeight=0.0082
         interpMode=RCIM_Cubic
         leaveTangent=-21.6376
         tangentMode=RCTM_User
         time=0.5
         value=0.4006
        }
        {arriveTangent=0.97, arriveTangentWeight=0.0022, interpMode=RCIM_Cubic, leaveTangent=0.97, tangentMode=RCTM_User, time=0.5065, value=0.373}
        {
         arriveTangent=18.2286
         arriveTangentWeight=0.0018
         interpMode=RCIM_Cubic
         leaveTangent=18.2286
         tangentMode=RCTM_User
         time=0.5119
         value=0.4667
        }
        {
         arriveTangent=63.439
         arriveTangentWeight=0.0033
         interpMode=RCIM_Cubic
         leaveTangent=63.439
         tangentMode=RCTM_User
         time=0.5217
         value=1.9909
        }
        {
         arriveTangent=9.2855
         arriveTangentWeight=0.0037
         interpMode=RCIM_Cubic
         leaveTangent=9.2855
         tangentMode=RCTM_User
         time=0.5328
         value=2.3668
        }
        {arriveTangent=-0.1145, arriveTangentWeight=0.1564, interpMode=RCIM_Cubic, leaveTangent=-0.1145, tangentMode=RCTM_User, time=1.0021, value=5}
       ]}}}
   ) {module=ScaleSpriteSize001}
   &ScaleSpriteSize(Uniform Curve Sprite Scale=DataInterface() {propertyValues {Curve {keys [
        {arriveTangent=26.7057, interpMode=RCIM_Cubic, leaveTangent=26.7057, tangentMode=RCTM_User, time=0, value=0}
        {arriveTangent=3.0432, interpMode=RCIM_Cubic, leaveTangent=3.0432, tangentMode=RCTM_User, time=0.1, value=1.25}
        {arriveTangent=4.0056, interpMode=RCIM_Cubic, leaveTangent=4.0056, tangentMode=RCTM_User, time=0.3, value=1.75}
        {arriveTangent=-0.3934, interpMode=RCIM_Cubic, leaveTangent=-0.3934, tangentMode=RCTM_User, time=0.5, value=2.1762}
        {arriveTangent=9.0575, interpMode=RCIM_Cubic, leaveTangent=9.0575, time=0.6027, value=6.5408}
        {interpMode=RCIM_Cubic, tangentMode=RCTM_User, time=0.9893, value=6.6081}
       ]}}}) {enabled=false}
   &CameraOffset()
  ]
  rendererClasses [&NiagaraSpriteRendererProperties]
  renderers [{rendererClass=&NiagaraSpriteRendererProperties}]
 }
 Halo {
  bEnabled=false
  emitterUpdateScript [
   &EmitterState(Life Cycle Mode=::NewEnumerator1, Loop Behavior=::NewEnumerator1, Loop Delay=0.1, Loop Duration=2.25, UseLoopDelay=false)
   &SpawnBurst_Instantaneous(Spawn Time=0.1, Use Spawn Probability=false, UseLoopCountLimit=false) {enabled=false}
   &SpawnRate(SpawnRate=5, Use Spawn Probability=false)
  ]
  particleSpawnScript [&InitializeParticle(
    Color=<rgba 1 0.25 0.25 0.5>
    Lifetime=0.25
    Position Offset=<0 0 12.5>
    Sprite Size Mode=::NewEnumerator3
    Uniform Sprite Size=25
    UsePositionOffset=true
    Write Material Random=true
   ) {module=InitializeParticle001}]
  particleUpdateScript [
   &ParticleState()
   &ScaleColor(
    Scale Alpha=&FloatFromCurve(FloatCurve=DataInterface() {propertyValues {Curve {keys [
         {arriveTangent=9.0725, interpMode=RCIM_Cubic, leaveTangent=9.0725, tangentMode=RCTM_User, time=-0.0069, value=0.1483}
         {arriveTangent=2.4835, interpMode=RCIM_Cubic, leaveTangent=2.4835, time=0.063, value=0.4346}
         {arriveTangent=-2.5249, interpMode=RCIM_Cubic, leaveTangent=-2.5249, tangentMode=RCTM_User, time=0.1639, value=0.5727}
         {arriveTangent=-0.0405, interpMode=RCIM_Cubic, leaveTangent=-0.0405, tangentMode=RCTM_User, time=0.9931, value=0.0989}
        ]}}})
    ScaleA=true
    ScaleRGB=true
   )
   &ScaleSpriteSize(Uniform Curve Sprite Scale=DataInterface() {propertyValues {Curve {keys [
        {arriveTangent=4.1785, interpMode=RCIM_Cubic, leaveTangent=4.1785, tangentMode=RCTM_User, time=0, value=0}
        {arriveTangent=0.9995, interpMode=RCIM_Cubic, leaveTangent=0.9995, tangentMode=RCTM_User, time=0.2223, value=0.6825}
        {arriveTangent=-0.0405, interpMode=RCIM_Cubic, leaveTangent=-0.0405, tangentMode=RCTM_User, time=1, value=1}
       ]}}})
   &CameraOffset(Camera Offset Amount=5)
  ]
  rendererClasses [&NiagaraSpriteRendererProperties]
  renderers [{rendererClass=&NiagaraSpriteRendererProperties}]
 }
 Heart_Shape {
  emitterUpdateScript [
   &EmitterState(Life Cycle Mode=::NewEnumerator1, Recalculate Duration Each Loop=true, UseLoopDelay=false)
   &SpawnBurst_Instantaneous(Use Spawn Probability=false, UseLoopCountLimit=false)
  ]
  particleSpawnScript [&InitializeParticle(
    Lifetime=5
    Mesh Scale Mode=::NewEnumerator3
    Mesh Uniform Scale=0.25
    Position Offset=<0 0 12.5>
    Sprite Size Mode=::NewEnumerator3
    Uniform Sprite Size=35
    UsePositionOffset=true
    Write Material Random=true
   ) {module=InitializeParticle001}]
  particleUpdateScript [
   &ParticleState()
   /SetVariables(Particles.Position=&Add_Vector(
     A=Linked() {linkedVariable {name=Particles.Position, type=NiagaraPosition}}
     B=&MakeVector(
      X=0
      Y=0
      Z=&FloatFromCurve(FloatCurve=DataInterface() {propertyValues {Curve {keys [
           {arriveTangent=4.1785, time=0, value=0}
           {arriveTangent=254.3651, interpMode=RCIM_Cubic, leaveTangent=254.3651, time=0.01, value=3.3564}
           {arriveTangent=3.4792, interpMode=RCIM_Cubic, leaveTangent=3.4792, tangentMode=RCTM_User, time=0.025, value=5}
           {arriveTangent=-254.3651, interpMode=RCIM_Cubic, leaveTangent=-254.3651, time=0.04, value=3.3564}
           <at 0.05 0>
           {arriveTangent=-0.0405, leaveTangent=-0.0405, time=1, value=0}
          ]}}})
     )
    )) {enabled=false, isSetParameters=true}
   &ScaleMeshSize(Scale Factor=&VectorFromCurve(VectorCurve=DataInterface() {propertyValues {
       XCurve {keys [<at 0 0>,<at 0.025 0.45>,<at 0.05 0.5>,<at 0.9 0.5>,<at 0.95 0.5>,<at 0.975 0.75>,<at 1 0>]}
       YCurve {keys [<at 0 0>,<at 0.025 0.8>,<at 0.05 0.5>,<at 0.9 0.5>,<at 0.95 0.5>,<at 0.98 0.6>,<at 1 0>]}
       ZCurve {keys [<at 0 0>,<at 0.05 0.6>,<at 0.075 0.5>,<at 0.925 0.5>,<at 0.95 0.5>,<at 0.975 0.55>,<at 1 0>]}
      }})) {enabled=false, module=ScaleMeshSize001}
   /SetVariables(Particles.Position=&Add_Vector(
     A=Linked() {linkedVariable {name=Particles.Initial.Position, type=NiagaraPosition}}
     B=&VectorFromCurve(VectorCurve=DataInterface() {propertyValues {
        XCurve {keys [<at 0 0>,<at 1 0>]}
        YCurve {keys [<at 0 0>,<at 1 0>]}
        ZCurve {keys [
          {arriveTangent=-29.7738, interpMode=RCIM_Cubic, leaveTangent=-29.7738, tangentMode=RCTM_User, time=0, value=0.8097}
          {arriveTangent=428.0804, interpMode=RCIM_Cubic, leaveTangent=428.0804, tangentMode=RCTM_User, time=0.0605, value=11.5343}
          {arriveTangent=25.3279, interpMode=RCIM_Cubic, leaveTangent=25.3279, tangentMode=RCTM_User, time=0.1502, value=10.5442}
          {arriveTangent=334.4314, interpMode=RCIM_Cubic, leaveTangent=334.4314, tangentMode=RCTM_User, time=0.2222, value=27.8481}
          {arriveTangent=25.038, interpMode=RCIM_Cubic, leaveTangent=25.038, time=0.3079, value=22.6366}
          {arriveTangent=148.1974, interpMode=RCIM_Cubic, leaveTangent=148.1974, tangentMode=RCTM_User, time=0.3986, value=32.2649}
          {arriveTangent=-552.3597, interpMode=RCIM_Cubic, leaveTangent=-552.3597, tangentMode=RCTM_User, time=0.4953, value=21.4375}
          {arriveTangent=817.7939, interpMode=RCIM_Cubic, leaveTangent=817.7939, tangentMode=RCTM_User, time=0.5052, value=22.3636}
          {arriveTangent=880.1193, interpMode=RCIM_Cubic, leaveTangent=880.1193, tangentMode=RCTM_User, time=0.512, value=26.6862}
          {arriveTangent=532.9858, interpMode=RCIM_Cubic, leaveTangent=532.9858, tangentMode=RCTM_User, time=0.5247, value=38.8145}
          {arriveTangent=123.7225, interpMode=RCIM_Cubic, leaveTangent=123.7225, tangentMode=RCTM_User, time=0.5502, value=44.8436}
          {arriveTangent=0.0095, interpMode=RCIM_Cubic, leaveTangent=0.0095, tangentMode=RCTM_User, time=1, value=100}
         ]}
       }})
    )) {isSetParameters=true}
   /SetVariables(Particles.Position=&Multiply_Vector(
     A=Linked() {linkedVariable {name=Particles.Position, type=NiagaraPosition}}
     B=&VectorFromCurve(VectorCurve=DataInterface() {propertyValues {
        XCurve {keys [<at 0 0>,<at 1 0>]}
        YCurve {keys [<at 0 0>,<at 1 0>]}
        ZCurve {keys [
          {arriveTangent=4.0573, time=0, value=1}
          {arriveTangent=1.6003, interpMode=RCIM_Cubic, leaveTangent=1.6003, tangentMode=RCTM_User, time=0.0137, value=1.132}
          {arriveTangent=-0.0109, leaveTangent=-0.0109, time=0.0341, value=0.9897}
          <at 0.0708 0.9932>
          <at 0.1732 1.0071>
          <at 0.2787 0.9921>
          <at 0.3887 1.007>
          {arriveTangent=-0.4632, interpMode=RCIM_Cubic, leaveTangent=-0.4632, time=0.4744, value=0.9965}
          {arriveTangent=0.073, interpMode=RCIM_Cubic, leaveTangent=0.073, tangentMode=RCTM_User, time=0.4781, value=0.9656}
          {arriveTangent=3.4942, interpMode=RCIM_Cubic, leaveTangent=3.4942, tangentMode=RCTM_User, time=0.4835, value=0.98}
          {arriveTangent=0.0078, interpMode=RCIM_Cubic, leaveTangent=0.0078, tangentMode=RCTM_User, time=0.5104, value=1.0154}
          {arriveTangent=0.0095, interpMode=RCIM_Cubic, leaveTangent=0.0095, tangentMode=RCTM_User, time=1, value=1}
         ]}
       }})
    )) {enabled=false, isSetParameters=true}
   &ScaleColor(
    Linear Color Curve=DataInterface() {propertyValues {
      AlphaCurve {keys [<at 0 1>,<at 1 1>]}
      BlueCurve {keys [<at 0 1>,<at 0.0687 0>,<at 1 1>]}
      GreenCurve {keys [<at 0 1>,<at 0.0687 0>,<at 1 1>]}
      RedCurve {keys [<at 0 1>,<at 0.0687 2>,<at 1 1>]}
     }}
    Scale Mode=::NewEnumerator2
   ) {enabled=false}
   &ScaleSpriteSize(
    Non-Uniform Curve Sprite Scale=DataInterface() {propertyValues {
      XCurve {keys [
        {arriveTangent=11.3603, interpMode=RCIM_Cubic, leaveTangent=11.3603, tangentMode=RCTM_User, time=0, value=0}
        {arriveTangent=24.145, interpMode=RCIM_Cubic, leaveTangent=24.145, tangentMode=RCTM_User, time=0.0108, value=0.4027}
        {arriveTangent=-1.4511, interpMode=RCIM_Cubic, leaveTangent=-1.4511, time=0.0149, value=0.6123}
        {arriveTangent=1.3167, interpMode=RCIM_Cubic, leaveTangent=1.3167, tangentMode=RCTM_User, time=0.0896, value=0.2884}
        {arriveTangent=0.9348, interpMode=RCIM_Cubic, leaveTangent=0.9348, tangentMode=RCTM_User, time=0.1844, value=0.5244}
        {arriveTangent=0.0399, interpMode=RCIM_Cubic, leaveTangent=0.0399, time=0.275, value=0.3067}
        {arriveTangent=2.2837, interpMode=RCIM_Cubic, leaveTangent=2.2837, tangentMode=RCTM_User, time=0.3708, value=0.5318}
        {arriveTangent=-5.1858, interpMode=RCIM_Cubic, leaveTangent=-5.1858, tangentMode=RCTM_User, time=0.4822, value=0.4132}
        {arriveTangent=-10.6961, interpMode=RCIM_Cubic, leaveTangent=-10.6961, time=0.4916, value=0.4449}
        {arriveTangent=-4.8294, interpMode=RCIM_Cubic, leaveTangent=-4.8294, tangentMode=RCTM_User, time=0.5019, value=0.2032}
        {arriveTangent=32.0657, interpMode=RCIM_Cubic, leaveTangent=32.0657, time=0.5115, value=0.384}
        {arriveTangent=-0.8515, interpMode=RCIM_Cubic, leaveTangent=-0.8515, tangentMode=RCTM_User, time=0.5289, value=1.0688}
        {arriveTangent=1.288, interpMode=RCIM_Cubic, leaveTangent=1.288, tangentMode=RCTM_User, time=0.5514, value=1.0878}
        {arriveTangent=-0.2651, interpMode=RCIM_Cubic, leaveTangent=-0.2651, tangentMode=RCTM_User, time=1.0021, value=1.3141}
       ]}
      YCurve {keys [
        {arriveTangent=19.4939, interpMode=RCIM_Cubic, leaveTangent=19.4939, tangentMode=RCTM_User, time=0, value=0}
        {arriveTangent=6.3156, interpMode=RCIM_Cubic, leaveTangent=6.3156, tangentMode=RCTM_User, time=0.0253, value=0.4972}
        {arriveTangent=-0.5169, interpMode=RCIM_Cubic, leaveTangent=-0.5169, tangentMode=RCTM_User, time=0.0894, value=0.3296}
        {arriveTangent=2.4886, interpMode=RCIM_Cubic, leaveTangent=2.4886, tangentMode=RCTM_User, time=0.1989, value=0.5205}
        {arriveTangent=-0.9078, interpMode=RCIM_Cubic, leaveTangent=-0.9078, tangentMode=RCTM_User, time=0.274, value=0.3596}
        {arriveTangent=1.6527, interpMode=RCIM_Cubic, leaveTangent=1.6527, tangentMode=RCTM_User, time=0.375, value=0.472}
        {arriveTangent=-5.4894, interpMode=RCIM_Cubic, leaveTangent=-5.4894, tangentMode=RCTM_User, time=0.4774, value=0.4609}
        {arriveTangent=-16.2714, interpMode=RCIM_Cubic, leaveTangent=-16.2714, tangentMode=RCTM_User, time=0.4947, value=0.2547}
        {arriveTangent=18.2286, interpMode=RCIM_Cubic, leaveTangent=18.2286, tangentMode=RCTM_User, time=0.5045, value=0.2483}
        {arriveTangent=2.1487, interpMode=RCIM_Cubic, leaveTangent=2.1487, tangentMode=RCTM_User, time=0.5217, value=1.0176}
        {arriveTangent=-0.1145, interpMode=RCIM_Cubic, leaveTangent=-0.1145, tangentMode=RCTM_User, time=1.0021, value=1.3236}
       ]}
     }}
    Scale Sprite Size Mode=::NewEnumerator3
   ) {enabled=false}
   &ScaleSpriteSize(
    Non-Uniform Curve Sprite Scale=DataInterface() {propertyValues {
      XCurve {keys [
        {arriveTangent=0.4109, interpMode=RCIM_Cubic, leaveTangent=0.4109, tangentMode=RCTM_User, time=0, value=0}
        {arriveTangent=-4.59, interpMode=RCIM_Cubic, leaveTangent=-4.59, tangentMode=RCTM_User, time=0.0767, value=0.9817}
        {arriveTangent=1.6878, interpMode=RCIM_Cubic, leaveTangent=1.6878, time=0.1446, value=0.5788}
        {arriveTangent=2.3227, interpMode=RCIM_Cubic, leaveTangent=2.3227, tangentMode=RCTM_User, time=0.2127, value=1.2112}
        {arriveTangent=3.3602, interpMode=RCIM_Cubic, leaveTangent=3.3602, time=0.3112, value=0.7479}
        {arriveTangent=11.4927, interpMode=RCIM_Cubic, leaveTangent=11.4927, tangentMode=RCTM_User, time=0.3985, value=1.8356}
        {arriveTangent=-5.1858, interpMode=RCIM_Cubic, leaveTangent=-5.1858, tangentMode=RCTM_User, time=0.4803, value=1.5082}
        {arriveTangent=-43.0804, interpMode=RCIM_Cubic, leaveTangent=-43.0804, time=0.4982, value=1.1242}
        {arriveTangent=-79.7231, interpMode=RCIM_Cubic, leaveTangent=-79.7231, tangentMode=RCTM_User, time=0.5034, value=0.5118}
        {arriveTangent=1.1492, interpMode=RCIM_Cubic, leaveTangent=1.1492, tangentMode=RCTM_User, time=0.509, value=0.5008}
        {arriveTangent=126.1387, interpMode=RCIM_Cubic, leaveTangent=126.1387, tangentMode=RCTM_User, time=0.517, value=1.0232}
        {arriveTangent=33.7355, interpMode=RCIM_Cubic, leaveTangent=33.7355, tangentMode=RCTM_User, time=0.5257, value=2.3016}
        {arriveTangent=27.7482, interpMode=RCIM_Cubic, leaveTangent=27.7482, tangentMode=RCTM_User, time=0.548, value=2.8878}
        {arriveTangent=-0.2651, interpMode=RCIM_Cubic, leaveTangent=-0.2651, tangentMode=RCTM_User, time=1.0021, value=5}
       ]}
      YCurve {keys [
        {arriveTangent=-0.2072, interpMode=RCIM_Cubic, leaveTangent=-0.2072, tangentMode=RCTM_User, time=0, value=0}
        {arriveTangent=18.5129, interpMode=RCIM_Cubic, leaveTangent=18.5129, tangentMode=RCTM_User, time=0.0442, value=0.5996}
        {arriveTangent=12.3238, interpMode=RCIM_Cubic, leaveTangent=12.3238, tangentMode=RCTM_User, time=0.0864, value=0.9004}
        {arriveTangent=1.8531, interpMode=RCIM_Cubic, leaveTangent=1.8531, time=0.152, value=0.5679}
        {arriveTangent=14.9653, interpMode=RCIM_Cubic, leaveTangent=14.9653, tangentMode=RCTM_User, time=0.2209, value=1.1496}
        {arriveTangent=-0.9078, interpMode=RCIM_Cubic, leaveTangent=-0.9078, tangentMode=RCTM_User, time=0.3172, value=0.7307}
        {arriveTangent=15.9193, interpMode=RCIM_Cubic, leaveTangent=15.9193, tangentMode=RCTM_User, time=0.4061, value=1.8527}
        {arriveTangent=-24.9715, interpMode=RCIM_Cubic, leaveTangent=-24.9715, tangentMode=RCTM_User, time=0.4755, value=1.5559}
        {arriveTangent=-21.6376, interpMode=RCIM_Cubic, leaveTangent=-21.6376, tangentMode=RCTM_User, time=0.5, value=0.4006}
        {arriveTangent=0.97, interpMode=RCIM_Cubic, leaveTangent=0.97, tangentMode=RCTM_User, time=0.5065, value=0.373}
        {arriveTangent=18.2286, interpMode=RCIM_Cubic, leaveTangent=18.2286, tangentMode=RCTM_User, time=0.5119, value=0.4667}
        {arriveTangent=63.439, interpMode=RCIM_Cubic, leaveTangent=63.439, tangentMode=RCTM_User, time=0.5217, value=2.2804}
        {arriveTangent=9.2855, interpMode=RCIM_Cubic, leaveTangent=9.2855, tangentMode=RCTM_User, time=0.5328, value=2.6563}
        {arriveTangent=-0.1145, interpMode=RCIM_Cubic, leaveTangent=-0.1145, tangentMode=RCTM_User, time=1.0021, value=5}
       ]}
     }}
    Scale Sprite Size Mode=::NewEnumerator3
   ) {module=ScaleSpriteSize001}
   &DynamicMaterialParameters(
    Index 0 Param 1=&FloatFromCurve(FloatCurve=DataInterface() {propertyValues {Curve {keys [
         {tangentMode=RCTM_User, time=0, value=2}
         {arriveTangent=-21.2792, interpMode=RCIM_Cubic, leaveTangent=-21.2792, tangentMode=RCTM_User, time=0.4767, value=2}
         {arriveTangent=-11.5581, interpMode=RCIM_Cubic, leaveTangent=-2.3625, tangentMode=RCTM_Break, time=0.5677, value=-0.0204}
         {arriveTangent=-0.2874, interpMode=RCIM_Cubic, leaveTangent=-0.2874, tangentMode=RCTM_User, time=0.9, value=-1}
        ]}}})
    Module.Param0WriteEnabled=true
    Module.Param1WriteEnabled=true
    Module.Param2WriteEnabled=true
    Module.Param3WriteEnabled=true
   )
   &SolveForcesAndVelocity(Clamp Velocity=false, Limit Acceleration=false)
  ]
  rendererClasses [&NiagaraMeshRendererProperties, &NiagaraSpriteRendererProperties]
  renderers [
   {rendererClass=&NiagaraMeshRendererProperties, rendererIndex=0}
   {rendererClass=&NiagaraSpriteRendererProperties, rendererIndex=1}
  ]
 }
 Heart_Sparks {
  simTarget=GPUComputeSim
  emitterUpdateScript [
   &EmitterState(Life Cycle Mode=::NewEnumerator1, Loop Delay=3, UseLoopDelay=false) {module=EmitterState001}
   &SpawnBurst_Instantaneous(Spawn Count=20, Use Spawn Probability=false, UseLoopCountLimit=false)
  ]
  particleSpawnScript [
   &InitializeParticle(
    Color Maximum=<rgba 0 0.8087 1 1>
    Color Minimum=<rgba 0.8236 0.1828 1 1>
    Color Mode=::NewEnumerator2
    Lifetime Max=1.5
    Lifetime Min=0.5
    Lifetime Mode=::NewEnumerator1
    Sprite Rotation Mode=::NewEnumerator2
    Sprite Size Mode=::NewEnumerator1
    Uniform Sprite Size Min=2.5
    UsePositionOffset=false
    Write Material Random=true
   ) {module=InitializeParticle001}
   &AddVelocity(Module.Use Falloff Distance=false, Override Seed=false, Velocity Mode=::NewEnumerator1, Velocity Speed=100)
   /Game/Effects/Niagara/Modules/Sprite/FN_SpriteUV(End Frame=4, Randomize Sprite UV scale=true)
   &ShapeLocation(MaskSpawnGroup=false, Offset=<0 0 20>, Offset Mode=::NewEnumerator0, Override Seed=false, Sphere Radius=0.05, UseOwnerScale=true)
  ]
  particleUpdateScript [
   &ParticleState()
   &Drag(Drag=7, Use Linear Drag=true, Use Rotational Drag=false)
   &GravityForce(Gravity=<0 0 50>)
   &SolveForcesAndVelocity(Clamp Velocity=false, Limit Acceleration=false)
   &ScaleColor(
    Scale Alpha=&FloatFromCurve(FloatCurve=DataInterface() {propertyValues {Curve {keys [
         {tangentMode=RCTM_User, time=0, value=1}
         {arriveTangent=-0.0405, interpMode=RCIM_Cubic, leaveTangent=-0.0405, tangentMode=RCTM_User, time=0.7009, value=1}
         {arriveTangent=-8.0089, interpMode=RCIM_Cubic, leaveTangent=-8.0089, tangentMode=RCTM_User, time=1, value=0}
        ]}}})
    Scale RGB=&VectorFromFloat(Value=&FloatFromCurve(FloatCurve=DataInterface() {propertyValues {Curve {keys [
          {tangentMode=RCTM_User, time=0, value=1}
          {arriveTangent=-0.0405, leaveTangent=-0.0405, time=0.7009, value=1}
          {arriveTangent=-8.0089, leaveTangent=-8.0089, time=1, value=1}
         ]}}}))
    ScaleA=true
    ScaleRGB=true
   ) {module=ScaleColor001}
   &ScaleSpriteSize(Uniform Curve Sprite Scale=DataInterface() {propertyValues {Curve {keys [
        {arriveTangent=4.1785, interpMode=RCIM_Cubic, leaveTangent=4.1785, tangentMode=RCTM_User, time=0, value=0}
        {arriveTangent=0.9995, interpMode=RCIM_Cubic, leaveTangent=0.9995, tangentMode=RCTM_User, time=0.1035, value=1.0837}
        {arriveTangent=-0.5114, interpMode=RCIM_Cubic, leaveTangent=-0.5114, time=0.2035, value=0.5723}
        {arriveTangent=-0.2557, interpMode=RCIM_Cubic, leaveTangent=-0.2557, time=0.3035, value=0.9814}
        <at 0.4035 0.5212 RCIM_Cubic>
        {arriveTangent=0.2557, interpMode=RCIM_Cubic, leaveTangent=0.2557, time=0.5035, value=0.9814}
        {arriveTangent=0.767, interpMode=RCIM_Cubic, leaveTangent=0.767, time=0.6035, value=0.5723}
        {arriveTangent=-0.2557, interpMode=RCIM_Cubic, leaveTangent=-0.2557, time=0.7035, value=1.1348}
        {arriveTangent=-0.5114, interpMode=RCIM_Cubic, leaveTangent=-0.5114, time=0.8035, value=0.5212}
        {arriveTangent=-2.6531, interpMode=RCIM_Cubic, leaveTangent=-2.6531, time=0.9035, value=1.0326}
        {arriveTangent=-3.4484, interpMode=RCIM_Cubic, leaveTangent=-3.4484, tangentMode=RCTM_User, time=1, value=0}
       ]}}}) {module=ScaleSpriteSize001}
  ]
  rendererClasses [&NiagaraSpriteRendererProperties]
  renderers [{rendererClass=&NiagaraSpriteRendererProperties}]
 }
 Heart_Sparks001 {
  simTarget=GPUComputeSim
  emitterUpdateScript [
   &EmitterState(Life Cycle Mode=::NewEnumerator1, Loop Delay=3, UseLoopDelay=false) {module=EmitterState001}
   &SpawnBurst_Instantaneous(Spawn Count=35, Spawn Time=2.5, Use Spawn Probability=false, UseLoopCountLimit=false)
  ]
  particleSpawnScript [
   &InitializeParticle(
    Color Maximum=<rgba 0 0.8087 1 1>
    Color Minimum=<rgba 0.8236 0.1828 1 1>
    Color Mode=::NewEnumerator2
    Lifetime Max=1.5
    Lifetime Min=0.5
    Lifetime Mode=::NewEnumerator1
    Position Offset=<0 0 20>
    Sprite Rotation Mode=::NewEnumerator2
    Sprite Size Mode=::NewEnumerator1
    Uniform Sprite Size Min=2.5
    UsePositionOffset=false
    Write Material Random=true
   ) {module=InitializeParticle001}
   &AddVelocity(Module.Use Falloff Distance=false, Override Seed=false, Velocity Mode=::NewEnumerator1, Velocity Speed=200)
   &FN_SpriteUV(End Frame=4, Randomize Sprite UV scale=true)
   &ShapeLocation(MaskSpawnGroup=false, Offset=<0 0 50>, Offset Mode=::NewEnumerator0, Override Seed=false, Sphere Radius=1, UseOwnerScale=true)
  ]
  particleUpdateScript [
   &ParticleState()
   &Drag(Drag=5, Use Linear Drag=true, Use Rotational Drag=false)
   &GravityForce(Gravity=<0 0 50>)
   &SolveForcesAndVelocity(Clamp Velocity=false, Limit Acceleration=false)
   &ScaleColor(
    Scale Alpha=&FloatFromCurve(FloatCurve=DataInterface() {propertyValues {Curve {keys [
         {tangentMode=RCTM_User, time=0, value=1}
         {arriveTangent=-0.0405, interpMode=RCIM_Cubic, leaveTangent=-0.0405, tangentMode=RCTM_User, time=0.7009, value=1}
         {arriveTangent=-8.0089, interpMode=RCIM_Cubic, leaveTangent=-8.0089, tangentMode=RCTM_User, time=1, value=0}
        ]}}})
    Scale RGB=&VectorFromFloat(Value=&FloatFromCurve(FloatCurve=DataInterface() {propertyValues {Curve {keys [
          {tangentMode=RCTM_User, time=0, value=1}
          {arriveTangent=-0.0405, leaveTangent=-0.0405, time=0.7009, value=1}
          {arriveTangent=-8.0089, leaveTangent=-8.0089, time=1, value=1}
         ]}}}))
    ScaleA=true
    ScaleRGB=true
   ) {module=ScaleColor001}
   &ScaleSpriteSize(Uniform Curve Sprite Scale=DataInterface() {propertyValues {Curve {keys [
        {arriveTangent=4.1785, interpMode=RCIM_Cubic, leaveTangent=4.1785, tangentMode=RCTM_User, time=0, value=0}
        {arriveTangent=0.9995, interpMode=RCIM_Cubic, leaveTangent=0.9995, tangentMode=RCTM_User, time=0.1035, value=1.0837}
        {arriveTangent=-0.5114, interpMode=RCIM_Cubic, leaveTangent=-0.5114, time=0.2035, value=0.5723}
        {arriveTangent=-0.2557, interpMode=RCIM_Cubic, leaveTangent=-0.2557, time=0.3035, value=0.9814}
        <at 0.4035 0.5212 RCIM_Cubic>
        {arriveTangent=0.2557, interpMode=RCIM_Cubic, leaveTangent=0.2557, time=0.5035, value=0.9814}
        {arriveTangent=0.767, interpMode=RCIM_Cubic, leaveTangent=0.767, time=0.6035, value=0.5723}
        {arriveTangent=-0.2557, interpMode=RCIM_Cubic, leaveTangent=-0.2557, time=0.7035, value=1.1348}
        {arriveTangent=-0.5114, interpMode=RCIM_Cubic, leaveTangent=-0.5114, time=0.8035, value=0.5212}
        {arriveTangent=-2.6531, interpMode=RCIM_Cubic, leaveTangent=-2.6531, time=0.9035, value=1.0326}
        {arriveTangent=-3.4484, interpMode=RCIM_Cubic, leaveTangent=-3.4484, tangentMode=RCTM_User, time=1, value=0}
       ]}}}) {module=ScaleSpriteSize001}
  ]
  rendererClasses [&NiagaraSpriteRendererProperties]
  renderers [{rendererClass=&NiagaraSpriteRendererProperties}]
 }
 Heart_Sparks002 {
  simTarget=GPUComputeSim
  emitterUpdateScript [
   &EmitterState(Life Cycle Mode=::NewEnumerator1, UseLoopDelay=false) {module=EmitterState001}
   &SpawnBurst_Instantaneous(Spawn Count=5, Use Spawn Probability=false, UseLoopCountLimit=false)
   &SpawnRate(SpawnRate=20, Use Spawn Probability=false)
  ]
  particleSpawnScript [
   &InitializeParticle(
    Color Maximum=<rgba 0 0.867 1 1>
    Color Minimum=<rgba 1 0.25 1 1>
    Color Mode=::NewEnumerator2
    Lifetime Max=1
    Lifetime Min=0.5
    Lifetime Mode=::NewEnumerator1
    Sprite Size Mode=::NewEnumerator1
    Uniform Sprite Size Min=2.5
    UsePositionOffset=false
    Write Material Random=true
   ) {module=InitializeParticle001}
   &ShapeLocation(
    MaskSpawnGroup=false
    Offset=<0 0 20>
    Offset Mode=::NewEnumerator0
    Override Seed=false
    Shape Origin=$Particles.Position
    Sphere Radius=15
    UseOwnerScale=true
   )
   /SetVariables(Particles.Position=&Add_Vector(
     A=Linked() {linkedVariable {name=Particles.Initial.Position, type=NiagaraPosition}}
     B=&VectorFromCurve(
      VectorCurve=DataInterface() {propertyValues {
        XCurve {keys [<at 0 0>,<at 1 0>]}
        YCurve {keys [<at 0 0>,<at 1 0>]}
        ZCurve {keys [
          {arriveTangent=335.8231, interpMode=RCIM_Cubic, leaveTangent=335.8231, tangentMode=RCTM_User, time=0, value=0.8097}
          {arriveTangent=428.0804, interpMode=RCIM_Cubic, leaveTangent=428.0804, tangentMode=RCTM_User, time=0.0437, value=11.5343}
          {arriveTangent=25.3279, interpMode=RCIM_Cubic, leaveTangent=25.3279, tangentMode=RCTM_User, time=0.1334, value=10.5442}
          {arriveTangent=334.4314, interpMode=RCIM_Cubic, leaveTangent=334.4314, tangentMode=RCTM_User, time=0.2126, value=27.8481}
          {arriveTangent=23.7436, interpMode=RCIM_Cubic, leaveTangent=23.7436, time=0.2983, value=22.6366}
          {arriveTangent=148.1974, interpMode=RCIM_Cubic, leaveTangent=148.1974, tangentMode=RCTM_User, time=0.3986, value=32.2649}
          {arriveTangent=-552.3597, interpMode=RCIM_Cubic, leaveTangent=-552.3597, tangentMode=RCTM_User, time=0.4953, value=21.4375}
          {arriveTangent=817.7939, interpMode=RCIM_Cubic, leaveTangent=817.7939, tangentMode=RCTM_User, time=0.5052, value=22.3636}
          {arriveTangent=880.1193, interpMode=RCIM_Cubic, leaveTangent=880.1193, tangentMode=RCTM_User, time=0.512, value=26.6862}
          {arriveTangent=532.9858, interpMode=RCIM_Cubic, leaveTangent=532.9858, tangentMode=RCTM_User, time=0.5247, value=38.8145}
          {arriveTangent=123.7225, interpMode=RCIM_Cubic, leaveTangent=123.7225, tangentMode=RCTM_User, time=0.5502, value=44.8436}
          {arriveTangent=0.0095, interpMode=RCIM_Cubic, leaveTangent=0.0095, tangentMode=RCTM_User, time=1, value=100}
         ]}
       }}
      CurveIndex=$Emitter.NormalizedLoopAge
     )
    )) {isSetParameters=true}
   &FN_SpriteUV(End Frame=4, Randomize Sprite UV scale=true) {enabled=false}
  ]
  particleUpdateScript [
   &ParticleState()
   &Drag(Drag=7.5, Use Linear Drag=true, Use Rotational Drag=false)
   &CurlNoiseForce(Noise Frequency=10, Noise Quality / Cost=::NewEnumerator0, Noise Strength=50, Pan Noise=false) {enabled=false}
   &SolveForcesAndVelocity(Clamp Velocity=false, Limit Acceleration=false)
   &ScaleColor(
    Scale Alpha=&FloatFromCurve(
     FloatCurve=DataInterface() {propertyValues {Curve {keys [
         {interpMode=RCIM_Cubic, tangentMode=RCTM_User, time=0, value=0}
         {arriveTangent=0.8913, interpMode=RCIM_Cubic, leaveTangent=0.8913, tangentMode=RCTM_User, time=0.1096, value=0.8167}
         {arriveTangent=-0.8637, interpMode=RCIM_Cubic, leaveTangent=-0.8637, tangentMode=RCTM_User, time=0.5, value=0.95}
         {arriveTangent=-6.0226, interpMode=RCIM_Cubic, leaveTangent=-6.0226, tangentMode=RCTM_User, time=0.8, value=0}
        ]}}}
     CurveIndex=$Emitter.NormalizedLoopAge
    )
    Scale RGB=&VectorFromFloat(Value=&FloatFromCurve(FloatCurve=DataInterface() {propertyValues {Curve {keys [
          {tangentMode=RCTM_User, time=0, value=1}
          {arriveTangent=-0.0405, leaveTangent=-0.0405, time=0.7009, value=1}
          {arriveTangent=-8.0089, leaveTangent=-8.0089, time=1, value=1}
         ]}}}))
    ScaleA=true
    ScaleRGB=true
   )
   &ScaleSpriteSize(
    Uniform Curve Index=$Emitter.NormalizedLoopAge
    Uniform Curve Sprite Scale=DataInterface() {propertyValues {Curve {keys [
        {arriveTangent=4.1785, interpMode=RCIM_Cubic, leaveTangent=4.1785, tangentMode=RCTM_User, time=0, value=0}
        {arriveTangent=0.9995, interpMode=RCIM_Cubic, leaveTangent=0.9995, tangentMode=RCTM_User, time=0.1, value=0.75}
        {arriveTangent=1.137, interpMode=RCIM_Cubic, leaveTangent=1.137, time=0.2035, value=0.5723}
        {arriveTangent=-0.2557, interpMode=RCIM_Cubic, leaveTangent=-0.2557, time=0.3035, value=0.9814}
        {arriveTangent=-1.4326, interpMode=RCIM_Cubic, leaveTangent=-1.4326, time=0.4035, value=0.5212}
        {arriveTangent=0.2557, interpMode=RCIM_Cubic, leaveTangent=0.2557, time=0.5, value=0.7}
        {arriveTangent=-2.75, interpMode=RCIM_Cubic, leaveTangent=-2.75, time=0.6035, value=0.5723}
        {arriveTangent=-2.4043, interpMode=RCIM_Cubic, leaveTangent=-2.4043, time=0.7, value=0.15}
        {arriveTangent=-0.5, interpMode=RCIM_Cubic, leaveTangent=-0.5, time=0.8, value=0.1}
        {arriveTangent=-0.5, interpMode=RCIM_Cubic, leaveTangent=-0.5, time=0.9, value=0.05}
        {arriveTangent=-3.4484, interpMode=RCIM_Cubic, leaveTangent=-3.4484, tangentMode=RCTM_User, time=1, value=0}
       ]}}}
   )
   &SpriteFacingAndAlignment(Sprite Alignment=<0 0 1>, WriteAlignment=false, WriteFacing=false)
   &JitterPosition(Jitter Amount=0.1)
  ]
  rendererClasses [&NiagaraSpriteRendererProperties]
  renderers [{rendererClass=&NiagaraSpriteRendererProperties}]
 }
 Heart_Sparks003 {
  emitterUpdateScript [
   &EmitterState(Life Cycle Mode=::NewEnumerator1, Loop Delay=2.45, Recalculate Duration Each Loop=true, UseLoopDelay=true) {module=EmitterState001}
   &SpawnBurst_Instantaneous(Spawn Count=50, Use Spawn Probability=false, UseLoopCountLimit=false)
  ]
  particleSpawnScript [
   &InitializeParticle(
    Lifetime Max=1.5
    Lifetime Min=0.5
    Lifetime Mode=::NewEnumerator1
    Sprite Rotation Mode=::NewEnumerator2
    Sprite Size Mode=::NewEnumerator1
    Uniform Sprite Size Max=15
    UsePositionOffset=false
    Write Material Random=true
   ) {module=InitializeParticle001}
   &AddVelocity(Module.Use Falloff Distance=false, Override Seed=false, Velocity Mode=::NewEnumerator1, Velocity Speed=200)
   &FN_SpriteUV(End Frame=4, Randomize Sprite UV scale=true)
   &ShapeLocation(MaskSpawnGroup=false, Offset=<0 0 40>, Offset Mode=::NewEnumerator0, Override Seed=false, Sphere Radius=5, UseOwnerScale=true)
  ]
  particleUpdateScript [
   &ParticleState()
   &Drag(Drag=5, Use Linear Drag=true, Use Rotational Drag=false)
   &CurlNoiseForce(Noise Frequency=25, Noise Strength=250, Pan Noise=false)
   &GravityForce(Gravity=<0 0 400>)
   &SolveForcesAndVelocity(Clamp Velocity=false, Limit Acceleration=false)
   &ScaleColor(
    Scale Alpha=&FloatFromCurve(FloatCurve=DataInterface() {propertyValues {Curve {keys [
         {tangentMode=RCTM_User, time=0, value=1}
         {arriveTangent=-0.0405, interpMode=RCIM_Cubic, leaveTangent=-0.0405, tangentMode=RCTM_User, time=0.7009, value=1}
         {arriveTangent=-8.0089, interpMode=RCIM_Cubic, leaveTangent=-8.0089, tangentMode=RCTM_User, time=1, value=0}
        ]}}})
    Scale RGB=&VectorFromFloat(Value=&FloatFromCurve(FloatCurve=DataInterface() {propertyValues {Curve {keys [
          {tangentMode=RCTM_User, time=0, value=1}
          {arriveTangent=-0.0405, leaveTangent=-0.0405, time=0.7009, value=1}
          {arriveTangent=-8.0089, leaveTangent=-8.0089, time=1, value=1}
         ]}}}))
    ScaleA=true
    ScaleRGB=true
   ) {module=ScaleColor001}
   &ScaleSpriteSize(Uniform Curve Sprite Scale=DataInterface() {propertyValues {Curve {keys [
        {arriveTangent=18.4349, interpMode=RCIM_Cubic, leaveTangent=18.4349, tangentMode=RCTM_User, time=0, value=0}
        {arriveTangent=0.9995, interpMode=RCIM_Cubic, leaveTangent=0.9995, tangentMode=RCTM_User, time=0.0415, value=0.65}
        {arriveTangent=-0.9885, interpMode=RCIM_Cubic, leaveTangent=-0.9885, time=0.4, value=0.85}
        {arriveTangent=-4.722, interpMode=RCIM_Cubic, leaveTangent=-4.722, tangentMode=RCTM_User, time=0.7426, value=-0.043}
       ]}}})
  ]
  rendererClasses [&NiagaraSpriteRendererProperties]
  renderers [{rendererClass=&NiagaraSpriteRendererProperties}]
 }
 Light {
  emitterUpdateScript [
   &EmitterState(Life Cycle Mode=::NewEnumerator1, UseLoopDelay=false)
   &SpawnBurst_Instantaneous(
    Spawn Group=1
    Spawn Probability=HlslExpression() {hlslExpression="Emitter.LoopCount == 0"}
    Spawn Time=0.03
    Use Spawn Probability=true
    UseLoopCountLimit=false
   ) {module=SpawnBurst_Instantaneous001}
  ]
  particleSpawnScript [&InitializeParticle(UsePositionOffset=false, Write Material Random=true)]
  particleUpdateScript [
   &ParticleState(Kill Particles When Lifetime Has Elapsed=false)
   &FN_Light_Attributes(
    Color=&ColorFromCurve(
     ColorCurve=DataInterface() {propertyValues {
       AlphaCurve {keys [<at 0.0024 1>,<at 1 0>]}
       BlueCurve {keys [<at 0.0038 0>,<at 0.1591 0.75>,<at 0.327 0>,<at 0.4948 0.75>,<at 0.6627 0>,<at 0.8218 0.75>,<at 1 0>]}
       GreenCurve {keys [<at 0.0038 0.25>,<at 0.1591 0>,<at 0.327 0.25>,<at 0.4948 0>,<at 0.6627 0.25>,<at 0.8218 0>,<at 1 0.25>]}
       RedCurve {keys [<at 0.0038 1>,<at 0.1591 1>,<at 0.327 1>,<at 0.4948 1>,<at 0.6627 1>,<at 0.8218 1>,<at 1 1>]}
      }}
     CurveIndex=$Emitter.NormalizedLoopAge
    )
    Create Color Attribute=true
    Create Exponent Attribute=true
    Create Position Attribute=true
    Create Radius Attribute=true
    Enable Particle Light Enable=true
    Exponent=5
    Position=<0 0 20>
    Radius=5
    Scale Color=5
   )
   &CameraOffset(Camera Offset Amount=50)
  ]
  rendererClasses [&NiagaraLightRendererProperties]
  renderers [{rendererClass=&NiagaraLightRendererProperties, rendererIndex=0}]
 }
}
summary {
 systemName=NS_HeartSign_Explosion_Emote
 emitters [
  {emitterName=Heart_Shape, rendererClasses [&NiagaraMeshRendererProperties, &NiagaraSpriteRendererProperties]}
  {bEnabled=false, emitterName=Halo, rendererClasses [&NiagaraSpriteRendererProperties]}
  {emitterName=Heart_Sparks, simTarget=GPUComputeSim, rendererClasses [&NiagaraSpriteRendererProperties]}
  {bEnabled=false, emitterName=Flare, rendererClasses [&NiagaraSpriteRendererProperties]}
  {emitterName=Flare001, rendererClasses [&NiagaraSpriteRendererProperties]}
  {emitterName=Heart_Sparks002, simTarget=GPUComputeSim, rendererClasses [&NiagaraSpriteRendererProperties]}
  {emitterName=Light, rendererClasses [&NiagaraLightRendererProperties]}
  {emitterName=Heart_Sparks003, rendererClasses [&NiagaraSpriteRendererProperties]}
  {emitterName=Heart_Sparks001, simTarget=GPUComputeSim, rendererClasses [&NiagaraSpriteRendererProperties]}
 ]
 userVariables [
  User.FrontendSpawn {type=NiagaraBool}
  User.RibbonIDOffset {type=NiagaraInt32}
  User.SkeletalMesh {
   type=NiagaraDataInterfaceSkeletalMesh
   defaultValue {
    struct=NiagaraExt_VariableValue_DataInterface
    value {dataInterface=&self:NiagaraDataInterfaceSkeletalMesh_1}
   }
  }
 ]
}
