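# Niagara effect for the main player's HeartSign idle emote, layering a flare, heart shape, sparks and light.
# Heart_Shape_Into continuously spawns short-lived mesh and sprite particles that fade, while GPU sparks jitter in curl noise.
# Flare and light share an orange-to-magenta color pulse driven by normalized loop age.
# User floats raise flare and heart along Z and scale their sprites, with the flare using only clamped FX_Size_Y.
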NiagaraSystem /Game/Animation/Game/MainPlayer/Emotes/HeartSign/FX/NS_HeartSign_Idle_Emote
dependencies {
 usedDataInterfaces [&NiagaraDataInterfaceColorCurve, &NiagaraDataInterfaceCurve, &NiagaraDataInterfaceVector2DCurve]
 usedDynamicInputs []
 usedModules [&SystemState]
}
emitters {
 Flare001 {
  emitterUpdateScript [
   &EmitterState(Delay First Loop Only=true, Inactive Response=::NewEnumerator1, Life Cycle Mode=::NewEnumerator1, Loop Delay=0.1, UseLoopDelay=true)
   &SpawnBurst_Instantaneous(Use Spawn Probability=false, UseLoopCountLimit=false)
  ]
  particleSpawnScript [&InitializeParticle(
    Color=<rgba 1 1 1 0.5>
    Lifetime=5
    Position Offset=<0 0 10>
    Position Offset Coordinate Space=::Local
    Sprite Size Mode=::NewEnumerator3
    Uniform Sprite Size=35
    UsePositionOffset=true
    Write Material Random=true
   ) {module=InitializeParticle001}]
  particleUpdateScript [
   &ParticleState()
   &ScaleColor(
    Curve Index=$Emitter.NormalizedLoopAge
    Linear Color Curve=DataInterface() {propertyValues {
      AlphaCurve {keys [<at 0.0077 0.5>,<at 0.4992 0.5>,<at 1 0.5>]}
      BlueCurve {keys [<at 0.0038 0>,<at 0.1591 0.75>,<at 0.327 0>,<at 0.4948 0.75>,<at 0.6627 0>,<at 0.8218 0.75>,<at 1 0>]}
      GreenCurve {keys [<at 0.0038 0.25>,<at 0.1591 0>,<at 0.327 0.25>,<at 0.4948 0>,<at 0.6627 0.25>,<at 0.8218 0>,<at 1 0.25>]}
      RedCurve {keys [<at 0.0038 1>,<at 0.1591 1>,<at 0.327 1>,<at 0.4948 1>,<at 0.6627 1>,<at 0.8218 1>,<at 1 1>]}
     }}
    Scale Mode=::NewEnumerator2
   )
   /SetVariables(Particles.Position=&Add_Vector(
     A=Linked() {linkedVariable {name=Particles.Initial.Position, type=NiagaraPosition}}
     B=&MakeVector(X=0, Y=0, Z=$User.FX_Position_Z)
    )) {isSetParameters=true}
   &ScaleSpriteSize(
    Scale Factor=&Multiply_Vector2D(
     A=&Vector2DFromFloat(Value=&ClampFloat(Float=$User.FX_Size_Y, Min=0.4, Max=0.8))
     B=<1.2 1.2>
    )
    Scale Sprite Size Mode=::NewEnumerator2
   ) {module=ScaleSpriteSize001}
   &ScaleSpriteSize(
    Non-Uniform Curve Scale=<1.5 1.5>
    Non-Uniform Curve Sprite Scale=DataInterface() {propertyValues {
      XCurve {keys [
        {arriveTangent=11.3603, interpMode=RCIM_Cubic, leaveTangent=11.3603, tangentMode=RCTM_User, time=-0.0013, value=0.4}
        {arriveTangent=-0.2899, interpMode=RCIM_Cubic, leaveTangent=-0.2899, tangentMode=RCTM_User, time=0.0326, value=0.5168}
        {arriveTangent=1.3167, interpMode=RCIM_Cubic, leaveTangent=1.3167, tangentMode=RCTM_User, time=0.1161, value=0.2884}
        {arriveTangent=0.9348, interpMode=RCIM_Cubic, leaveTangent=0.9348, tangentMode=RCTM_User, time=0.2121, value=0.5168}
        {arriveTangent=0.9885, interpMode=RCIM_Cubic, leaveTangent=0.9885, tangentMode=RCTM_User, time=0.3191, value=0.2884}
        {arriveTangent=2.2837, interpMode=RCIM_Cubic, leaveTangent=2.2837, tangentMode=RCTM_User, time=0.4099, value=0.5168}
        {arriveTangent=1.0107, interpMode=RCIM_Cubic, leaveTangent=1.0107, tangentMode=RCTM_User, time=0.5257, value=0.2884}
        {arriveTangent=1.8896, interpMode=RCIM_Cubic, leaveTangent=1.8896, tangentMode=RCTM_User, time=0.6178, value=0.5168}
        {interpMode=RCIM_Cubic, tangentMode=RCTM_User, time=0.7402, value=0.2884}
        {interpMode=RCIM_Cubic, tangentMode=RCTM_User, time=0.8286, value=0.5168}
        {interpMode=RCIM_Cubic, tangentMode=RCTM_User, time=0.9497, value=0.2884}
        {arriveTangent=3.5216, interpMode=RCIM_Cubic, leaveTangent=3.5216, tangentMode=RCTM_User, time=1, value=0.4}
       ]}
      YCurve {keys [
        {arriveTangent=5.3614, interpMode=RCIM_Cubic, leaveTangent=5.3614, tangentMode=RCTM_User, time=-0.0013, value=0.4}
        {arriveTangent=0.0966, interpMode=RCIM_Cubic, leaveTangent=0.0966, tangentMode=RCTM_User, time=0.0404, value=0.5033}
        {arriveTangent=-0.5169, interpMode=RCIM_Cubic, leaveTangent=-0.5169, tangentMode=RCTM_User, time=0.1159, value=0.3296}
        {arriveTangent=2.4886, interpMode=RCIM_Cubic, leaveTangent=2.4886, tangentMode=RCTM_User, time=0.2267, value=0.5033}
        {arriveTangent=-0.9078, interpMode=RCIM_Cubic, leaveTangent=-0.9078, tangentMode=RCTM_User, time=0.3182, value=0.3296}
        {arriveTangent=1.6527, interpMode=RCIM_Cubic, leaveTangent=1.6527, tangentMode=RCTM_User, time=0.4406, value=0.5033}
        {interpMode=RCIM_Cubic, tangentMode=RCTM_User, time=0.5245, value=0.3296}
        {arriveTangent=1.3012, interpMode=RCIM_Cubic, leaveTangent=1.3012, tangentMode=RCTM_User, time=0.6443, value=0.5064}
        {interpMode=RCIM_Cubic, tangentMode=RCTM_User, time=0.739, value=0.3296}
        {arriveTangent=1.77, interpMode=RCIM_Cubic, leaveTangent=1.77, tangentMode=RCTM_User, time=0.8538, value=0.5033}
        {arriveTangent=-0.4066, interpMode=RCIM_Cubic, leaveTangent=-0.4066, tangentMode=RCTM_User, time=0.9472, value=0.3296}
        {arriveTangent=2.4395, interpMode=RCIM_Cubic, leaveTangent=2.4395, tangentMode=RCTM_User, time=1, value=0.4}
       ]}
     }}
    Scale Sprite Size Mode=::NewEnumerator3
   ) {enabled=false}
   &CameraOffset()
  ]
  rendererClasses [&NiagaraSpriteRendererProperties]
  renderers [{rendererClass=&NiagaraSpriteRendererProperties}]
 }
 Heart_Shape_Into {
  emitterUpdateScript [
   &EmitterState(Life Cycle Mode=::NewEnumerator1, UseLoopDelay=false)
   &SpawnBurst_Instantaneous(Spawn Probability=HlslExpression() {hlslExpression="Emitter.LoopCount == 0"}, Use Spawn Probability=false, UseLoopCountLimit=false)
   &SpawnRate(SpawnRate=24, Use Spawn Probability=false)
  ]
  particleSpawnScript [
   &InitializeParticle(
    Lifetime=0.5
    Mesh Scale Mode=::NewEnumerator3
    Mesh Uniform Scale=0.25
    Position Offset=<0 0 10>
    Sprite Size Mode=::NewEnumerator3
    Uniform Sprite Size=35
    UsePositionOffset=true
    Write Material Random=true
   ) {module=InitializeParticle001}
   &CameraOffset(Camera Offset Amount=1)
  ]
  particleUpdateScript [
   &ParticleState()
   &SystemLocation(MaskSpawnGroup=false, Offset=<0 0 18.5>)
   /SetVariables(Particles.Position=&Add_Vector(
     A=Linked() {linkedVariable {name=Particles.Initial.Position, type=NiagaraPosition}}
     B=&MakeVector(X=0, Y=0, Z=$User.FX_Position_Z)
    )) {isSetParameters=true}
   &ScaleSpriteSize(
    Scale Factor=&MakeVector2D(X=$User.FX_Size_X, Y=$User.FX_Size_Y)
    Scale Sprite Size Mode=::NewEnumerator2
   ) {module=ScaleSpriteSize001}
   &DynamicMaterialParameters(
    Index 0 Param 1=10
    Module.Param0WriteEnabled=true
    Module.Param1WriteEnabled=true
    Module.Param2WriteEnabled=true
    Module.Param3WriteEnabled=true
   )
   &SolveForcesAndVelocity(Clamp Velocity=false, Limit Acceleration=false)
   &ScaleColor(
    Linear Color Curve=DataInterface() {propertyValues {
      AlphaCurve {keys [<at 0 1>,<at 0.0328 1>,<at 0.2075 0.5>,<at 0.8192 0>]}
      BlueCurve {keys [<at 0 1>,<at 0.4759 1>,<at 1 1>]}
      GreenCurve {keys [<at 0 1>,<at 0.4759 1>,<at 1 1>]}
      RedCurve {keys [<at 0 1>,<at 0.4759 1>,<at 1 1>]}
     }}
    Scale Mode=::NewEnumerator2
   )
   &ScaleSpriteSize(Uniform Curve Sprite Scale=DataInterface() {propertyValues {Curve {keys [
        {tangentMode=RCTM_User, time=0, value=1}
        {arriveTangent=-0.0405, interpMode=RCIM_Cubic, leaveTangent=-0.0405, tangentMode=RCTM_User, time=0.1, value=0.9756}
        {arriveTangent=-11.7449, interpMode=RCIM_Cubic, leaveTangent=-11.7449, tangentMode=RCTM_User, time=0.4, value=-0.0244}
       ]}}}) {module=ScaleSpriteSize002}
  ]
  rendererClasses [&NiagaraMeshRendererProperties, &NiagaraSpriteRendererProperties]
  renderers [
   {rendererClass=&NiagaraMeshRendererProperties, rendererIndex=0}
   {rendererClass=&NiagaraSpriteRendererProperties, rendererIndex=1}
  ]
 }
 Heart_Sparks002 {
  simTarget=GPUComputeSim
  emitterUpdateScript [
   &EmitterState(Life Cycle Mode=::NewEnumerator1, UseLoopDelay=false) {module=EmitterState001}
   &SpawnBurst_Instantaneous(Spawn Count=5, Use Spawn Probability=false, UseLoopCountLimit=false)
   &SpawnRate(SpawnRate=20, Use Spawn Probability=false)
  ]
  particleSpawnScript [
   &InitializeParticle(
    Color Maximum=<rgba 0 0.867 1 1>
    Color Minimum=<rgba 1 0.25 1 1>
    Color Mode=::NewEnumerator2
    Lifetime Max=1
    Lifetime Min=0.5
    Lifetime Mode=::NewEnumerator1
    Sprite Size Mode=::NewEnumerator1
    Uniform Sprite Size Min=2.5
    UsePositionOffset=false
    Write Material Random=true
   ) {module=InitializeParticle001}
   /Game/Effects/Niagara/Modules/Sprite/FN_SpriteUV(End Frame=4, Randomize Sprite UV scale=true) {enabled=false}
   &ShapeLocation(
    MaskSpawnGroup=false
    Offset=<0 0 20>
    Offset Coordinate Space=::Local
    Offset Mode=::NewEnumerator0
    Override Seed=false
    Sphere Radius=15
    UseOwnerScale=true
   )
  ]
  particleUpdateScript [
   &ParticleState()
   &Drag(Drag=7.5, Use Linear Drag=true, Use Rotational Drag=false)
   &CurlNoiseForce(Noise Frequency=10, Noise Quality / Cost=::NewEnumerator0, Noise Strength=50, Pan Noise=false)
   &SolveForcesAndVelocity(Clamp Velocity=false, Limit Acceleration=false)
   &ScaleColor(
    Scale Alpha=&FloatFromCurve(FloatCurve=DataInterface() {propertyValues {Curve {keys [
         {interpMode=RCIM_Cubic, tangentMode=RCTM_User, time=0, value=0}
         {arriveTangent=0.8913, interpMode=RCIM_Cubic, leaveTangent=0.8913, tangentMode=RCTM_User, time=0.1096, value=0.8167}
         {arriveTangent=-0.8637, interpMode=RCIM_Cubic, leaveTangent=-0.8637, tangentMode=RCTM_User, time=0.5, value=0.95}
         {arriveTangent=-6.0226, interpMode=RCIM_Cubic, leaveTangent=-6.0226, tangentMode=RCTM_User, time=0.8, value=0}
        ]}}})
    Scale RGB=&VectorFromFloat(Value=&FloatFromCurve(FloatCurve=DataInterface() {propertyValues {Curve {keys [
          {tangentMode=RCTM_User, time=0, value=1}
          {arriveTangent=-0.0405, leaveTangent=-0.0405, time=0.7009, value=1}
          {arriveTangent=-8.0089, leaveTangent=-8.0089, time=1, value=1}
         ]}}}))
    ScaleA=true
    ScaleRGB=true
   )
   &ScaleSpriteSize(Uniform Curve Sprite Scale=DataInterface() {propertyValues {Curve {keys [
        {arriveTangent=4.1785, interpMode=RCIM_Cubic, leaveTangent=4.1785, tangentMode=RCTM_User, time=0, value=0}
        {arriveTangent=0.9995, interpMode=RCIM_Cubic, leaveTangent=0.9995, tangentMode=RCTM_User, time=0.1, value=0.75}
        {arriveTangent=1.137, interpMode=RCIM_Cubic, leaveTangent=1.137, time=0.2035, value=0.5723}
        {arriveTangent=-0.2557, interpMode=RCIM_Cubic, leaveTangent=-0.2557, time=0.3035, value=0.9814}
        {arriveTangent=-1.4326, interpMode=RCIM_Cubic, leaveTangent=-1.4326, time=0.4035, value=0.5212}
        {arriveTangent=0.2557, interpMode=RCIM_Cubic, leaveTangent=0.2557, time=0.5, value=0.7}
        {arriveTangent=-2.75, interpMode=RCIM_Cubic, leaveTangent=-2.75, time=0.6035, value=0.5723}
        {arriveTangent=-2.4043, interpMode=RCIM_Cubic, leaveTangent=-2.4043, time=0.7, value=0.15}
        {arriveTangent=-0.5, interpMode=RCIM_Cubic, leaveTangent=-0.5, time=0.8, value=0.1}
        {arriveTangent=-0.5, interpMode=RCIM_Cubic, leaveTangent=-0.5, time=0.9, value=0.05}
        {arriveTangent=-3.4484, interpMode=RCIM_Cubic, leaveTangent=-3.4484, tangentMode=RCTM_User, time=1, value=0}
       ]}}})
   &SpriteFacingAndAlignment(Sprite Alignment=<0 0 1>, WriteAlignment=false, WriteFacing=false)
   &JitterPosition(Jitter Amount=0.1)
  ]
  rendererClasses [&NiagaraSpriteRendererProperties]
  renderers [{rendererClass=&NiagaraSpriteRendererProperties}]
 }
 Light {
  emitterUpdateScript [
   &EmitterState(Inactive Response=::NewEnumerator1, Life Cycle Mode=::NewEnumerator1, UseLoopDelay=false)
   &SpawnBurst_Instantaneous(
    Spawn Group=1
    Spawn Probability=HlslExpression() {hlslExpression="Emitter.LoopCount == 0"}
    Spawn Time=0.03
    Use Spawn Probability=true
    UseLoopCountLimit=false
   ) {module=SpawnBurst_Instantaneous001}
  ]
  particleUpdateScript [
   &ParticleState(Kill Particles When Lifetime Has Elapsed=false)
   &FN_Light_Attributes(
    Color=&ColorFromCurve(
     ColorCurve=DataInterface() {propertyValues {
       AlphaCurve {keys [<at 0.4992 1>]}
       BlueCurve {keys [<at 0.0038 0>,<at 0.1591 0.75>,<at 0.327 0>,<at 0.4948 0.75>,<at 0.6627 0>,<at 0.8218 0.75>,<at 1 0>]}
       GreenCurve {keys [<at 0.0038 0.25>,<at 0.1591 0>,<at 0.327 0.25>,<at 0.4948 0>,<at 0.6627 0.25>,<at 0.8218 0>,<at 1 0.25>]}
       RedCurve {keys [<at 0.0038 1>,<at 0.1591 1>,<at 0.327 1>,<at 0.4948 1>,<at 0.6627 1>,<at 0.8218 1>,<at 1 1>]}
      }}
     CurveIndex=$Emitter.NormalizedLoopAge
    )
    Create Color Attribute=true
    Create Exponent Attribute=true
    Create Position Attribute=true
    Create Radius Attribute=true
    Enable Particle Light Enable=true
    Exponent=5
    Position=<0 0 20>
    Radius=5
    Scale Color=3
   )
   &CameraOffset(Camera Offset Amount=50)
  ]
  rendererClasses [&NiagaraLightRendererProperties]
  renderers [{rendererClass=&NiagaraLightRendererProperties, rendererIndex=0}]
 }
}
summary {
 systemName=NS_HeartSign_Idle_Emote
 emitters [
  {emitterName=Flare001, rendererClasses [&NiagaraSpriteRendererProperties]}
  {emitterName=Heart_Sparks002, simTarget=GPUComputeSim, rendererClasses [&NiagaraSpriteRendererProperties]}
  {emitterName=Light, rendererClasses [&NiagaraLightRendererProperties]}
  {
   emitterName=Heart_Shape_Into
   rendererClasses [&NiagaraMeshRendererProperties, &NiagaraSpriteRendererProperties]
  }
 ]
 userVariables [
  User.FrontendSpawn {type=NiagaraBool}
  User.FX_Position_Z {type=NiagaraFloat, defaultValue=0, description=""}
  User.FX_Size_X {type=NiagaraFloat, defaultValue=0, description=""}
  User.FX_Size_Y {type=NiagaraFloat, defaultValue=0, description=""}
  User.RibbonIDOffset {type=NiagaraInt32}
 ]
}
